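# Translucent, lit Niagara sprite material for spike streaks in the mud interaction gameplay cue effects.
# An explosion streak texture's red channel multiplies particle RGB to give the colour.
# A second chain scales particle alpha by Param_1 and the streak mask, then by a saturated mask minus inverted alpha.
# That subtraction erodes the streak from its dimmer areas as particle alpha falls, rather than fading it evenly.
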Material /FNGameplayCues/Effects/MudInteraction/Materials/M_Mud_Spikes
graph {
 properties {bUsedWithNiagaraSprites=true, blendMode=BLEND_Translucent, pixelDepthOffsetMode=PDOM_Legacy, usedShadingModels=Default Lit}
 nodes [
  Constant() {r=0.5}
  n1 = ParticleColor()
  n2 = TextureSample() {texture=/Game/Effects/Fort_Effects/Gameplay/SmokeDustExplosions/Textures/T_Explosion_Streak_02}
  Multiply(A=@n1.RGB, B=@n2.R)
  n3 = ScalarParameter() {defaultValue=3, parameterName=Param_1}
  n4 = OneMinus(None=@n1.A)
  Multiply(A=Multiply(A=Multiply(A=@n1.A, B=@n3) {constB=10}, B=@n2.R), B=Saturate(None=Subtract(A=@n2.R, B=@n4)))
  ScalarParameter()
 ]
}
